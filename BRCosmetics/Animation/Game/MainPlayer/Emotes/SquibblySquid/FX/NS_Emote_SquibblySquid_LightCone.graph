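# Niagara effect for the SquibblySquid emote that produces light cones with glow, ring, rainbow flare and drifting dust at sockets read through a Socket Reader.
# Light_Cones bursts spawn groups into mesh and sprite layers, using time-of-day colors and age curves for flash and pulsing scale.
# Camera-to-axis dot products fade the front cone layers and fade and resize Rainbow_Flare by view angle.
# Front-end modifiers retune sizes, alpha and dust rate.
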NiagaraSystem /BRCosmetics/Animation/Game/MainPlayer/Emotes/SquibblySquid/FX/NS_Emote_SquibblySquid_LightCone
dependencies {
 usedDataInterfaces [
  &NiagaraDataInterfaceCamera
  &NiagaraDataInterfaceCurve
  &NiagaraDataInterfaceVector2DCurve
  &NiagaraDataInterfaceVectorCurve
  &NiagaraDataInterfaceSocketReader
 ]
 usedDynamicInputs [&RandomRangeVector, &FNTimeOfDayColor]
 usedModules [
  &self:SystemSpawnScript.SystemScriptSource.SystemScriptGraph.NiagaraNodeAssignment_1.SetVariables_#1
  &SystemState
  &self:SystemSpawnScript.SystemScriptSource.SystemScriptGraph.NiagaraNodeAssignment_3.SetVariables_#2
 ]
}
emitters {
 Dust {
  emitterUpdateScript [
   &EmitterState(Delay First Loop Only=true, Life Cycle Mode=::NewEnumerator1, Loop Delay=0.5, Loop Duration=6.55, UseLoopDelay=true)
   &SpawnRate(
    SpawnRate=&RandomRangeFloat(
     Minimum=&FN_FrontEndFloatModifier(Gameplay=1, Front End=2)
     Maximum=&FN_FrontEndFloatModifier(Gameplay=5, Front End=8)
     Evaluation Type=::NewEnumerator1
    )
    Use Spawn Probability=false
   )
   &SpawnBurst_Instantaneous(Spawn Count=6, Use Spawn Probability=false, UseLoopCountLimit=false)
  ]
  particleSpawnScript [
   &InitializeParticle(
    Color=<rgba 0.8781 0.6468 0.486 0.25>
    Lifetime Max=1.5
    Lifetime Min=0.8
    Lifetime Mode=::NewEnumerator1
    Position Offset=<35 -13 -7>
    Sprite Size Mode=::NewEnumerator1
    Uniform Sprite Size Max=1
    Uniform Sprite Size Min=0.5
    UsePositionOffset=true
    Write Material Random=true
   )
   &PartitionParticles()
   &SocketLocation(
    SamplePosOffset=true
    Sampled Position Offset=<-5 0 0>
    Socket Index=$Particles.PartitionParticles.Partition
    Socket Reader=$System.Socket Reader
    Socket Sampling Mode=::NewEnumerator3
    VelocitySpeedLimit=false
   )
   &AddVelocity(
    Cone Axis=<-1 0 0>
    Module.UseConeFalloff=false
    Module.UseVelDistribution=false
    Override Seed=false
    Velocity Mode=::NewEnumerator2
    Velocity Speed=&RandomRangeFloat(Minimum=20, Maximum=30)
   )
   &ApplyInitialForces()
  ]
  particleUpdateScript [
   &ParticleState()
   &Drag(Drag=1.5, Use Linear Drag=true, Use Rotational Drag=true)
   &ScaleSpriteSize(
    Uniform Curve Scale=&FN_FrontEndFloatModifier(Gameplay=1, Front End=1.1)
    Uniform Curve Sprite Scale=DataInterface() {propertyValues {Curve {keys [
        {arriveTangent=0.0255, interpMode=RCIM_Cubic, leaveTangent=0.0255, tangentMode=RCTM_User, time=0.1564, value=0.0471}
        {interpMode=RCIM_Cubic, tangentMode=RCTM_User, time=0.4614, value=1}
        {interpMode=RCIM_Cubic, tangentMode=RCTM_User, time=0.6827, value=1}
        {arriveTangent=-0.0681, interpMode=RCIM_Cubic, leaveTangent=-0.0681, tangentMode=RCTM_User, time=1, value=0}
       ]}}}
   )
   &SolveForcesAndVelocity(Clamp Velocity=false, Limit Acceleration=false)
   &ScaleColor(
    Scale Alpha=&FloatFromCurve(
     FloatCurve=DataInterface() {propertyValues {Curve {keys [
         {arriveTangent=0.0255, interpMode=RCIM_Cubic, leaveTangent=0.0255, tangentMode=RCTM_User, time=0.1564, value=0.0471}
         {interpMode=RCIM_Cubic, tangentMode=RCTM_User, time=0.3214, value=0.9173}
         {arriveTangent=-1.3518, interpMode=RCIM_Cubic, leaveTangent=-1.3518, time=0.6587, value=0.244}
         {arriveTangent=-0.0681, interpMode=RCIM_Cubic, leaveTangent=-0.0681, tangentMode=RCTM_User, time=1, value=0}
        ]}}}
     Scale Curve=0.5
    )
    ScaleA=true
    ScaleRGB=true
   )
  ]
  rendererClasses [&NiagaraSpriteRendererProperties]
  renderers [{rendererClass=&NiagaraSpriteRendererProperties}]
 }
 Glow {rendererClasses [], renderers []}
 Light {rendererClasses [], renderers []}
 Light_Cones {
  emitterUpdateScript [
   &EmitterState(Life Cycle Mode=::NewEnumerator1, Loop Behavior=::NewEnumerator1, Loop Duration Mode=::NewEnumerator1, UseLoopDelay=false)
   &SpawnBurst_Instantaneous(Spawn Count=2, Use Spawn Probability=false, UseLoopCountLimit=false)
   &SpawnBurst_Instantaneous(Spawn Count=2, Spawn Group=1, Use Spawn Probability=false, UseLoopCountLimit=false) {module=SpawnBurst_Instantaneous001}
   &SpawnBurst_Instantaneous(Spawn Count=2, Spawn Group=2, Use Spawn Probability=false, UseLoopCountLimit=false) {module=SpawnBurst_Instantaneous002}
   &SpawnBurst_Instantaneous(Spawn Count=2, Spawn Group=3, Use Spawn Probability=false, UseLoopCountLimit=false) {module=SpawnBurst_Instantaneous003}
   &SpawnBurst_Instantaneous(Spawn Count=2, Spawn Group=4, Use Spawn Probability=false, UseLoopCountLimit=false) {module=SpawnBurst_Instantaneous006}
   &SpawnBurst_Instantaneous(Spawn Count=2, Spawn Group=5, Use Spawn Probability=false, UseLoopCountLimit=false) {module=SpawnBurst_Instantaneous007}
  ]
  particleSpawnScript [
   /SetVariables(Particles.VisibilityTag=&SpawnGroup()) {isSetParameters=true}
   &InitializeParticle(
    Color Mode=::NewEnumerator0
    Position Offset=&MergeHelper_VectorSwitch(0=<15 0 0>, 1=<15 0 0>, 2=<16.9 0 0>, 3=<16.9 0 0>, 4=<17.3 0 0>, Index=&SpawnGroup(), Num=::NewEnumerator3)
    UsePositionOffset=true
    Write Material Random=true
   )
   /SetVariables(
    Particles.Initial Light Soft Color:LinearColor=&FN_Select_Color_from_Time_Of_Day(Morning=<rgba 1 0.621 0.1255 0.36>, Day=<rgba 1 0.621 0.1255 0.28>, Evening=<rgba 1 0.621 0.1255 0.34>, Night=<rgba 1 0.621 0.1255 0.7>)
    Particles.Initial Light Soft Size:Vector3f=<0.637 -0.25 0.637>
   ) {isSetParameters=true}
   /SetVariables(
    Particles.Initial Light Wide Soft Color:LinearColor=&FN_Select_Color_from_Time_Of_Day(Morning=<rgba 2 0.9325 0.118 1>, Day=<rgba 2 0.9325 0.118 0.9>, Evening=<rgba 2 0.9325 0.118 1>, Night=<rgba 2 0.8655 0 0.75>)
    Particles.Initial Light Wide Soft Size:Vector3f=<0.82 -0.35 0.82>
   ) {isSetParameters=true}
   /SetVariables(
    Particles.Initial Light Front Color:LinearColor=&FN_Select_Color_from_Time_Of_Day(
     Morning=<rgba 1 0.5724 0.3263 0.12>
     Day=<rgba 1 0.5724 0.3263 0.06>
     Evening=<rgba 1 0.5724 0.3263 0.077>
     Night=<rgba 1.2 0.7284 0.3916 0.15>
    )
    Particles.Initial Light Front Size:Vector3f=<0.475 -0.35 0.475>
   ) {isSetParameters=true}
   /SetVariables(
    Particles.Initial Light Wide Front Color:LinearColor=&FN_Select_Color_from_Time_Of_Day(
     Morning=<rgba 1 0.5724 0.3263 0.12>
     Day=<rgba 1 0.5724 0.3263 0.06>
     Evening=<rgba 1 0.5724 0.3263 0.075>
     Night=<rgba 1.5 0.7926 0.4894 0.15>
    )
    Particles.Initial Light Wide Front Size:Vector3f=<0.6 -0.5 0.6>
   ) {isSetParameters=true}
   /SetVariables(
    Particles.Initial Glow Color:LinearColor=&FN_Select_Color_from_Time_Of_Day(Morning=<rgba 1 0.6064 0.1108 0.85>, Day=<rgba 1.1 0.667 0.1219 0.6>, Evening=<rgba 1 0.6064 0.1108 0.85>, Night=<rgba 1 0.6064 0.1108 0.8>)
    Particles.Initial Glow Size:Vector2f=&Vector2DFromFloat(Value=&FN_FrontEndFloatModifier(Gameplay=8.5, Front End=7.5))
   ) {isSetParameters=true}
   /SetVariables(
    Particles.Initial Ring Color:LinearColor=&FN_Select_Color_from_Time_Of_Day(Morning=<rgba 1 0.574 0.2065 1>, Day=<rgba 1.4 0.8036 0.289 1>, Evening=<rgba 1 0.574 0.2065 1>, Night=<rgba 1 0.574 0.2065 1>)
    Particles.Initial Ring Size:Vector2f=&Vector2DFromFloat(Value=&FN_FrontEndFloatModifier(Gameplay=8, Front End=6.5))
   ) {isSetParameters=true}
   &InitialMeshOrientation(
    Mesh Axis To Orient=&MergeHelper_VectorSwitch(0=<1 0 0>, 1=<1 0 0>, 2=<1 0 0>, 3=<1 0 0>, 4=<0 0 0>, Index=&SpawnGroup(), Num=::NewEnumerator3)
    Mesh Orientation Mode=::NewEnumerator2
    Orientation Vector=&MergeHelper_VectorSwitch(0=<0 1 0>, 1=<0 1 0>, 2=<0 1 0>, 3=<0 1 0>, 4=<0 0 0>, Index=&SpawnGroup(), Num=::NewEnumerator3)
    Use Rotation Vector=false
   )
   &CameraOffset(Camera Offset Amount=&MergeHelper_FloatSwitch(0=0, 1=0, 2=0, 3=0, 4=2, Index=&SpawnGroup(), Num=::NewEnumerator3))
   &PartitionParticles()
  ]
  particleUpdateScript [
   &ParticleState(Kill Particles When Lifetime Has Elapsed=false)
   /SetVariables(
    Particles.Scale Light Soft Color:LinearColor=&MakeLinearColorFromVectorAndFloat(
     "Vector (RGB)"=&MakeVectorFromLinearColorRGB(Color=/Niagara/DynamicInputs/Add/Add_LinearColor(
       A:LinearColor=$Particles.Initial.Initial Light Soft Color
       B:LinearColor=/Niagara/DynamicInputs/LinearColor/MakeLinearColorFromHSV(
        HSV From:NiagaraInt32=1
        HSV:Vector3f=?
        Hue:NiagaraFloat=0
        Saturation:NiagaraFloat=0
        Value:NiagaraFloat=&FloatFromCurve(
         FloatCurve=DataInterface() {propertyValues {Curve {keys [
             {interpMode=RCIM_Cubic, tangentMode=RCTM_User, time=0.007, value=0.7}
             {arriveTangent=-9.6851, interpMode=RCIM_Cubic, leaveTangent=-9.6851, tangentMode=RCTM_User, time=0.2, value=0}
            ]}}}
         CurveIndex=$Particles.Age
        )
        Alpha:NiagaraFloat=0
       )
      ))
     "Float (Alpha)"=&MakeFloatFromVector(
      Vector=&VectorFromFloat(Value=&FloatFromCurve(FloatCurve=DataInterface() {propertyValues {Curve {keys [<at 0 1.7>,<at 0.1 1>,<at 1 0.9903>]}}}, CurveIndex=$Particles.Age, Scale Curve=0.2))
      Channel=::NewEnumerator0
     )
    )
    Particles.Scale Light Soft Size:Vector3f=&Multiply_Vector(
     A=$Particles.Initial.Initial Light Soft Size
     B=&VectorFromCurve(
      VectorCurve=DataInterface() {propertyValues {
        XCurve {keys [<at 0 1>]}
        YCurve {keys [
          {arriveTangent=1.1872, interpMode=RCIM_Cubic, leaveTangent=1.1872, tangentMode=RCTM_User, time=0, value=0.5}
          {interpMode=RCIM_Cubic, leaveTangent=-4, time=1, value=1}
         ]}
        ZCurve {keys [<at 0 1>]}
       }}
      CurveIndex=$Particles.Age
      Scale Curve=&VectorFromFloat(Value=&FloatFromCurve(
        FloatCurve=DataInterface() {propertyValues {Curve {
           defaultValue=1
           keys [
            {arriveTangent=0.4534, interpMode=RCIM_Cubic, leaveTangent=0.4534, tangentMode=RCTM_User, time=0, value=0.2}
            {arriveTangent=5.5602, arriveTangentWeight=0.1333, interpMode=RCIM_Cubic, leaveTangent=1.5521, tangentMode=RCTM_Break, time=0.4, value=1.5}
            {arriveTangent=-0.2857, interpMode=RCIM_Cubic, leaveTangent=-0.2857, time=0.55, value=0.95}
            {arriveTangent=0.2222, interpMode=RCIM_Cubic, leaveTangent=0.2222, time=0.75, value=1.4}
            {arriveTangent=-0.1429, interpMode=RCIM_Cubic, leaveTangent=-0.1429, time=1, value=1.05}
            {arriveTangent=0.3333, interpMode=RCIM_Cubic, leaveTangent=0.3333, time=1.1, value=1.35}
            <at 1.3 1.15 RCIM_Cubic>
            <at 1.5 1.35 RCIM_Cubic>
           ]
          }}}
        CurveIndex=$Particles.Age
       ))
     )
    )
   ) {isSetParameters=true}
   /SetVariables(
    Particles.Scale Light Wide Soft Color:LinearColor=&MakeLinearColorFromVectorAndFloat(
     "Vector (RGB)"=&MakeVectorFromLinearColorRGB(Color=&Add_LinearColor(
       A:LinearColor=$Particles.Initial.Initial Light Wide Soft Color
       B:LinearColor=&MakeLinearColorFromHSV(
        HSV From:NiagaraInt32=1
        HSV:Vector3f=?
        Hue:NiagaraFloat=0
        Saturation:NiagaraFloat=0
        Value:NiagaraFloat=&FloatFromCurve(
         FloatCurve=DataInterface() {propertyValues {Curve {keys [
             {interpMode=RCIM_Cubic, tangentMode=RCTM_User, time=0.007, value=0.7}
             {arriveTangent=-10.7289, interpMode=RCIM_Cubic, leaveTangent=-10.7289, tangentMode=RCTM_User, time=0.2, value=0}
            ]}}}
         CurveIndex=$Particles.Age
        )
        Alpha:NiagaraFloat=0
       )
      ))
     "Float (Alpha)"=&MakeFloatFromVector(
      Vector=&VectorFromFloat(Value=&FloatFromCurve(FloatCurve=DataInterface() {propertyValues {Curve {keys [<at 0 1.7>,<at 0.1 1>,<at 1 1>]}}}, CurveIndex=$Particles.Age, Scale Curve=0.05))
      Channel=::NewEnumerator0
     )
    )
    Particles.Scale Light Wide Soft Size:Vector3f=&Multiply_Vector(
     A=$Particles.Initial.Initial Light Wide Soft Size
     B=&VectorFromCurve(
      VectorCurve=DataInterface() {propertyValues {
        XCurve {keys [<at 0 1>]}
        YCurve {keys [
          {arriveTangent=1.1872, interpMode=RCIM_Cubic, leaveTangent=1.1872, tangentMode=RCTM_User, time=0, value=0.5}
          {interpMode=RCIM_Cubic, leaveTangent=-4, time=1, value=1}
         ]}
        ZCurve {keys [<at 0 1>]}
       }}
      CurveIndex=$Particles.Age
      Scale Curve=&VectorFromFloat(Value=&FloatFromCurve(
        FloatCurve=DataInterface() {propertyValues {Curve {
           defaultValue=1
           keys [
            {arriveTangent=0.4534, interpMode=RCIM_Cubic, leaveTangent=0.4534, tangentMode=RCTM_User, time=0, value=0.2}
            {arriveTangent=5.5602, arriveTangentWeight=0.1333, interpMode=RCIM_Cubic, leaveTangent=1.5521, tangentMode=RCTM_Break, time=0.4, value=1.5}
            {arriveTangent=-0.2857, interpMode=RCIM_Cubic, leaveTangent=-0.2857, time=0.55, value=0.95}
            {arriveTangent=0.2222, interpMode=RCIM_Cubic, leaveTangent=0.2222, time=0.75, value=1.4}
            {arriveTangent=-0.1429, interpMode=RCIM_Cubic, leaveTangent=-0.1429, time=1, value=1.05}
            {arriveTangent=0.3333, interpMode=RCIM_Cubic, leaveTangent=0.3333, time=1.1, value=1.35}
            <at 1.3 1.15 RCIM_Cubic>
            <at 1.5 1.35 RCIM_Cubic>
           ]
          }}}
        CurveIndex=$Particles.Age
       ))
     )
    )
   ) {isSetParameters=true}
   /SetVariables(
    Particles.Scale Light Front Color:LinearColor=&MakeLinearColorFromVectorAndFloat(
     "Vector (RGB)"=&MakeVectorFromLinearColorRGB(Color=&Add_LinearColor(
       A:LinearColor=$Particles.Initial.Initial Light Front Color
       B:LinearColor=&MakeLinearColorFromHSV(
        HSV From:NiagaraInt32=1
        HSV:Vector3f=?
        Hue:NiagaraFloat=0
        Saturation:NiagaraFloat=0
        Value:NiagaraFloat=&FloatFromCurve(
         FloatCurve=DataInterface() {propertyValues {Curve {keys [
             {interpMode=RCIM_Cubic, tangentMode=RCTM_User, time=0.007, value=0.7}
             {arriveTangent=-127.0625, interpMode=RCIM_Cubic, leaveTangent=-127.0625, tangentMode=RCTM_User, time=0.02, value=0}
            ]}}}
         CurveIndex=$Particles.Age
        )
        Alpha:NiagaraFloat=1
       )
      ))
     "Float (Alpha)"=&Multiply_Float(
      A=&FloatFromCurve(
       FloatCurve=DataInterface() {propertyValues {Curve {keys [
           {arriveTangent=-9.5511, interpMode=RCIM_Cubic, leaveTangent=-9.5511, tangentMode=RCTM_User, time=0, value=1.7}
           {arriveTangent=-0.7, interpMode=RCIM_Cubic, leaveTangent=-0.7, time=0.1, value=1}
           <at 1 1 RCIM_Cubic>
          ]}}}
       CurveIndex=$Particles.Age
       Scale Curve=&FN_FrontEndFloatModifier(Gameplay=0.5, Front End=0.3)
      )
      B=&RemapRange(
       Input Value To Remap=&Abs_Float(Float:NiagaraFloat=&DotProduct(
         Vector A=&ReturnCameraVector(Camera Vectors=::NewEnumerator4, Coordinate Space=::Simulation, Origin Position=$Particles.Position)
         Vector B=&ReturnMeshOrientationAxis(Mesh Orientation Axis=::YAxis, Mesh Orientation Quat=$Particles.MeshOrientation)
        ))
       Input Min=0.1
       Input Max=1
       Output Min=0.08
       Output Max=0.4
      )
     )
    )
    Particles.Scale Light Front Size:Vector3f=&Multiply_Vector(
     A=$Particles.Initial.Initial Light Front Size
     B=&VectorFromCurve(
      VectorCurve=DataInterface() {propertyValues {
        XCurve {keys [<at 0 1>]}
        YCurve {keys [
          {arriveTangent=1.1872, interpMode=RCIM_Cubic, leaveTangent=1.1872, tangentMode=RCTM_User, time=0, value=0.5}
          {interpMode=RCIM_Cubic, leaveTangent=-4, time=1, value=1}
         ]}
        ZCurve {keys [<at 0 1>]}
       }}
      CurveIndex=$Particles.Age
      Scale Curve=&VectorFromFloat(Value=&FloatFromCurve(
        FloatCurve=DataInterface() {propertyValues {Curve {
           defaultValue=1
           keys [
            {arriveTangent=0.4534, interpMode=RCIM_Cubic, leaveTangent=0.4534, tangentMode=RCTM_User, time=0, value=0.2}
            {arriveTangent=5.5602, arriveTangentWeight=0.1333, interpMode=RCIM_Cubic, leaveTangent=1.5521, tangentMode=RCTM_Break, time=0.4, value=1.5}
            {arriveTangent=-0.2857, interpMode=RCIM_Cubic, leaveTangent=-0.2857, time=0.55, value=0.95}
            {arriveTangent=0.2222, interpMode=RCIM_Cubic, leaveTangent=0.2222, time=0.75, value=1.4}
            {arriveTangent=-0.1429, interpMode=RCIM_Cubic, leaveTangent=-0.1429, time=1, value=1.05}
            {arriveTangent=0.3333, interpMode=RCIM_Cubic, leaveTangent=0.3333, time=1.1, value=1.35}
            <at 1.3 1.15 RCIM_Cubic>
            <at 1.5 1.35 RCIM_Cubic>
           ]
          }}}
        CurveIndex=$Particles.Age
       ))
     )
    )
   ) {isSetParameters=true}
   /SetVariables(
    Particles.Scale Light Wide Front Color:LinearColor=&MakeLinearColorFromVectorAndFloat(
     "Vector (RGB)"=&MakeVectorFromLinearColorRGB(Color=&Add_LinearColor(
       A:LinearColor=$Particles.Initial.Initial Light Wide Front Color
       B:LinearColor=&MakeLinearColorFromHSV(
        HSV From:NiagaraInt32=1
        HSV:Vector3f=?
        Hue:NiagaraFloat=0
        Saturation:NiagaraFloat=0
        Value:NiagaraFloat=&FloatFromCurve(FloatCurve=DataInterface() {propertyValues {Curve {keys [
             {interpMode=RCIM_Cubic, tangentMode=RCTM_User, time=0.007, value=0.7}
             {arriveTangent=-127.0625, interpMode=RCIM_Cubic, leaveTangent=-127.0625, tangentMode=RCTM_User, time=0.02, value=0}
            ]}}})
        Alpha:NiagaraFloat=1
       )
      ))
     "Float (Alpha)"=&Multiply_Float(
      A=&FloatFromCurve(
       FloatCurve=DataInterface() {propertyValues {Curve {keys [<at 0 1.7>,<at 0.1 1>,<at 1 1>]}}}
       CurveIndex=$Particles.Lifetime
       Scale Curve=0.025
      )
      B=&RemapRange(
       Input Value To Remap=&Abs_Float(Float:NiagaraFloat=&DotProduct(
         Vector A=&ReturnCameraVector(Camera Vectors=::NewEnumerator4, Coordinate Space=::Simulation, Origin Position=$Particles.Position)
         Vector B=&ReturnMeshOrientationAxis(Mesh Orientation Axis=::YAxis, Mesh Orientation Quat=$Particles.MeshOrientation)
        ))
       Input Min=0
       Input Max=1
       Output Min=0.08
       Output Max=0.7
      )
     )
    )
    Particles.Scale Light Wide Front Size:Vector3f=&Multiply_Vector(
     A=$Particles.Initial.Initial Light Wide Front Size
     B=&VectorFromCurve(
      VectorCurve=DataInterface() {propertyValues {
        XCurve {keys [<at 0 1>]}
        YCurve {keys [
          {arriveTangent=1.1872, interpMode=RCIM_Cubic, leaveTangent=1.1872, tangentMode=RCTM_User, time=0, value=0.5}
          {interpMode=RCIM_Cubic, leaveTangent=-4, time=1, value=1}
         ]}
        ZCurve {keys [<at 0 1>]}
       }}
      CurveIndex=$Particles.Age
      Scale Curve=&VectorFromFloat(Value=&FloatFromCurve(
        FloatCurve=DataInterface() {propertyValues {Curve {
           defaultValue=1
           keys [
            {arriveTangent=0.4534, interpMode=RCIM_Cubic, leaveTangent=0.4534, tangentMode=RCTM_User, time=0, value=0.2}
            {arriveTangent=5.5602, arriveTangentWeight=0.1333, interpMode=RCIM_Cubic, leaveTangent=1.5521, tangentMode=RCTM_Break, time=0.4, value=1.5}
            {arriveTangent=-0.2857, interpMode=RCIM_Cubic, leaveTangent=-0.2857, time=0.55, value=0.95}
            {arriveTangent=0.2222, interpMode=RCIM_Cubic, leaveTangent=0.2222, time=0.75, value=1.4}
            {arriveTangent=-0.1429, interpMode=RCIM_Cubic, leaveTangent=-0.1429, time=1, value=1.05}
            {arriveTangent=0.3333, interpMode=RCIM_Cubic, leaveTangent=0.3333, time=1.1, value=1.35}
            <at 1.3 1.15 RCIM_Cubic>
            <at 1.5 1.35 RCIM_Cubic>
           ]
          }}}
        CurveIndex=$Particles.Age
       ))
     )
    )
   ) {isSetParameters=true}
   /SetVariables(
    Particles.Scale Glow Color:LinearColor=&MakeLinearColorFromVectorAndFloat(
     "Vector (RGB)"=&MakeVectorFromLinearColorRGB(Color=$Particles.Initial.Initial Glow Color)
     "Float (Alpha)"=&Multiply_Float(
      A=&FloatFromCurve(
       FloatCurve=DataInterface() {propertyValues {Curve {keys [{interpMode=RCIM_Cubic, leaveTangent=-1.0212, time=0.65, value=0.9}]}}}
       CurveIndex=$Particles.Age
       Scale Curve=&FN_FrontEndFloatModifier(Gameplay=1, Front End=3)
      )
      B=&FloatFromCurve(
       FloatCurve=DataInterface() {propertyValues {Curve {
          defaultValue=1
          keys [
           {arriveTangent=0.4534, interpMode=RCIM_Cubic, leaveTangent=0.4534, tangentMode=RCTM_User, time=0, value=0.05}
           {arriveTangent=5.5602, arriveTangentWeight=0.1333, interpMode=RCIM_Cubic, leaveTangent=1.5521, tangentMode=RCTM_Break, time=0.1155, value=1.1}
           {arriveTangent=-0.4989, interpMode=RCIM_Cubic, leaveTangent=-0.4989, time=0.2655, value=0.55}
           {arriveTangent=0.0564, interpMode=RCIM_Cubic, leaveTangent=0.0564, time=0.4655, value=0.9254}
           {arriveTangent=-0.1429, interpMode=RCIM_Cubic, leaveTangent=-0.1429, time=0.7155, value=0.5754}
           {arriveTangent=0.3333, interpMode=RCIM_Cubic, leaveTangent=0.3333, time=0.8155, value=0.8754}
           {arriveTangent=0.3115, interpMode=RCIM_Cubic, leaveTangent=0.3115, time=1.0155, value=0.6754}
           <at 1.2155 1 RCIM_Cubic>
          ]
         }}}
       CurveIndex=$Particles.Age
      )
     )
    )
    Particles.Scale Glow Size:Vector2f=&Multiply_Vector2D(
     A=$Particles.Initial Glow Size
     B=&Vector2DFromFloat(Value=&FloatFromCurve(
       FloatCurve=DataInterface() {propertyValues {Curve {
          defaultValue=1
          keys [
           {arriveTangent=0.4534, interpMode=RCIM_Cubic, leaveTangent=0.4534, tangentMode=RCTM_User, time=0, value=0.05}
           {arriveTangent=5.5602, arriveTangentWeight=0.1333, interpMode=RCIM_Cubic, leaveTangent=1.5521, tangentMode=RCTM_Break, time=0.4, value=1.5}
           {arriveTangent=-0.2857, interpMode=RCIM_Cubic, leaveTangent=-0.2857, time=0.55, value=0.95}
           {arriveTangent=0.2222, interpMode=RCIM_Cubic, leaveTangent=0.2222, time=0.75, value=1.4}
           {arriveTangent=-0.1429, interpMode=RCIM_Cubic, leaveTangent=-0.1429, time=1, value=1.05}
           {arriveTangent=0.3333, interpMode=RCIM_Cubic, leaveTangent=0.3333, time=1.1, value=1.35}
           <at 1.3 1.15 RCIM_Cubic>
           <at 1.5 1.35 RCIM_Cubic>
          ]
         }}}
       CurveIndex=$Particles.Age
       Scale Curve=0.6
      ))
    )
   ) {isSetParameters=true}
   /SetVariables(Particles.Scale Ring Color:LinearColor=&MakeLinearColorFromVectorAndFloat(
     "Vector (RGB)"=&MakeVectorFromLinearColorRGB(Color=$Particles.Initial Ring Color)
     "Float (Alpha)"=&FloatFromCurve(
      FloatCurve=DataInterface() {propertyValues {Curve {
         defaultValue=1
         keys [
          {arriveTangent=0.4534, interpMode=RCIM_Cubic, leaveTangent=0.4534, tangentMode=RCTM_User, time=0, value=0.05}
          {arriveTangent=5.5602, arriveTangentWeight=0.1333, interpMode=RCIM_Cubic, leaveTangent=1.5521, tangentMode=RCTM_Break, time=0.4, value=1.5}
          {arriveTangent=-0.2857, interpMode=RCIM_Cubic, leaveTangent=-0.2857, time=0.55, value=0.95}
          {arriveTangent=0.2222, interpMode=RCIM_Cubic, leaveTangent=0.2222, time=0.75, value=1.4}
          {arriveTangent=-0.1429, interpMode=RCIM_Cubic, leaveTangent=-0.1429, time=1, value=1.05}
          {arriveTangent=0.3333, interpMode=RCIM_Cubic, leaveTangent=0.3333, time=1.1, value=1.35}
          <at 1.3 1.15 RCIM_Cubic>
          <at 1.5 1.35 RCIM_Cubic>
         ]
        }}}
      CurveIndex=$Particles.Age
     )
    )) {isSetParameters=true}
   &SocketLocation(
    SamplePosOffset=false
    Socket Index=$Particles.PartitionParticles.Partition
    Socket Reader=$System.Socket Reader
    Socket Sampling Mode=::NewEnumerator3
    VelocitySpeedLimit=false
   )
  ]
  rendererClasses [&NiagaraMeshRendererProperties, &NiagaraSpriteRendererProperties]
  renderers [
   {rendererClass=&NiagaraMeshRendererProperties, rendererIndex=0}
   {rendererClass=&NiagaraMeshRendererProperties, rendererIndex=1}
   {rendererClass=&NiagaraMeshRendererProperties, rendererIndex=2}
   {rendererClass=&NiagaraMeshRendererProperties, rendererIndex=3}
   {rendererClass=&NiagaraSpriteRendererProperties, rendererIndex=4}
   {rendererClass=&NiagaraSpriteRendererProperties, rendererIndex=5}
  ]
 }
 Rainbow_Flare {
  emitterUpdateScript [
   &EmitterState(Life Cycle Mode=::NewEnumerator1, Loop Behavior=::NewEnumerator1, Loop Duration Mode=::NewEnumerator1, UseLoopDelay=false)
   &SpawnBurst_Instantaneous(Spawn Count=2, Use Spawn Probability=false, UseLoopCountLimit=false)
  ]
  particleSpawnScript [
   &InitializeParticle(
    Color Mode=::NewEnumerator0
    Lifetime=6.616
    Sprite Size Mode=::NewEnumerator3
    Uniform Sprite Size=15
    UsePositionOffset=false
    Write Material Random=true
   )
   &PartitionParticles()
   &SpriteFacingAndAlignment(Sprite Alignment=<1 0 0>, Sprite Facing=<-1 0 0>, WriteAlignment=false, WriteFacing=true)
   &SocketLocation(
    SamplePosOffset=true
    Sampled Position Coordinate Space:ENiagaraCoordinateSpace=::Local
    Sampled Position Offset=&MergeHelper_VectorSwitch(0=<-15 -5 0>, 1=<-15 15 0>, 2=?, 3=?, Index=$Particles.PartitionParticles.Partition, Num=::NewEnumerator0)
    Sampled Position Offset Coordinate Space:ENiagaraSkelMeshTransforms=::NewEnumerator2
    Socket Index=$Particles.PartitionParticles.Partition
    Socket Reader=$System.Socket Reader
    Socket Sampling Mode=::NewEnumerator3
    VelocitySpeedLimit=false
   )
  ]
  particleUpdateScript [
   &ParticleState(Kill Particles When Lifetime Has Elapsed=false)
   &ScaleColor(
    Scale Alpha=&Multiply_Float(
     A=&Multiply_Float(
      A=1
      B=&RemapRange(
       Input Value To Remap=&Abs_Float(Float:NiagaraFloat=&DotProduct(
         Vector A=&ReturnCameraVector(Camera Vectors=::NewEnumerator4, Coordinate Space=::Simulation, Origin Position=$Particles.Position)
         Vector B=$Particles.SpriteFacing
        ))
       Input Min=0.7
       Input Max=1
       Output Min=0
       Output Max=1
      )
     )
     B=&FN_FrontEndFloatModifier(Gameplay=1.2, Front End=0.8)
    )
    ScaleA=true
    ScaleRGB=true
   )
   &ScaleSpriteSize(
    Scale Factor=&Vector2DFromCurve(
     Vector2Curve=DataInterface() {propertyValues {XCurve {keys [<at 0 0>,<at 1 1>]}, YCurve {keys [<at 1 1>]}}}
     CurveIndex=&Multiply_Float(
      A=1
      B=&RemapRange(
       Input Value To Remap=&Abs_Float(Float:NiagaraFloat=&DotProduct(
         Vector A=&ReturnCameraVector(Camera Vectors=::NewEnumerator4, Coordinate Space=::Simulation, Origin Position=$Particles.Position)
         Vector B=$Particles.SpriteFacing
        ))
       Input Min=0.35
       Input Max=1
       Output Min=0
       Output Max=1
      )
     )
     ScaleCurve=<1 1>
    )
    Scale Sprite Size Mode=::NewEnumerator2
   )
  ]
  rendererClasses [&NiagaraSpriteRendererProperties]
  renderers [{rendererClass=&NiagaraSpriteRendererProperties}]
 }
}
summary {
 systemName=NS_Emote_SquibblySquid_LightCone
 emitters [
  {emitterName=Light, rendererClasses []}
  {emitterName=Glow, rendererClasses []}
  {emitterName=Rainbow_Flare, rendererClasses [&NiagaraSpriteRendererProperties]}
  {emitterName=Dust, rendererClasses [&NiagaraSpriteRendererProperties]}
  {emitterName=Light_Cones, rendererClasses [&NiagaraMeshRendererProperties, &NiagaraSpriteRendererProperties]}
 ]
 userVariables [
  User.bIsFrontEnd {type=NiagaraBool}
  User.bIsFrontEndPreview {type=NiagaraBool}
 ]
}
